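# Modular Control Rig tail module that builds a gimbal base control, a chained set of FK controls with offsets, and orient-only space switches from connector-resolved joints.
# Virtual FK joints follow the offset controls to pose the tail, and controls can also be set back from the joints.
# A rotate proxy adds its per-update Euler delta to FK controls selected by Position and Spread channels, with optional debug-line feedback.
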ControlRigBlueprint /EpicControlRig/Modules/Sidekick/Tail/CRM_Epic_Sidekick_Tail_v01
graphs [
 ConstructionGraph {nodes [
   RigUnit_PrepareForExecution = &RigUnit_PrepareForExecution()
   RigVMFunction_Sequence = RigVMAggregateNode(ExecuteContext=@RigUnit_PrepareForExecution.ExecutePin)
   Create Chain Fk Joints = RigVMFunctionReferenceNode(ExecuteContext=@RigVMFunction_Sequence.C)
   Create Gimbal Control = RigVMFunctionReferenceNode(ExecuteContext=@RigVMFunction_Sequence.D)
   > Create Chain Fk Controls = RigVMFunctionReferenceNode()
   > Create Fk Offset Controls = RigVMFunctionReferenceNode()
   > Create Proxy Controls = RigVMFunctionReferenceNode()
   Create Space Switches = RigVMFunctionReferenceNode(ExecuteContext=@RigVMFunction_Sequence.E)
   n1 = /Script/ControlRig.RigUnit_ResolveArrayConnector(Connector=(Type=Connector,Name="Tail"), SkipSocket=False) {is=struct RigVMUnitNode}
   RigVMVariableNode(Value=@n1.Result, ExecuteContext=@RigVMFunction_Sequence.A) {variable=Tail Chain Joints}
   > RigVMVariableNode(Value=RigVMDispatchNode(Index=0, Array=@n1.Result) {template="DISPATCH_RigVMDispatch_ArrayGetAtIndex(in Array,in Index,out Element)"}.Element) {variable=Connector Tail Base}
   > RigVMVariableNode(Value=/Script/ControlRig.RigUnit_ResolveConnector(Connector=(Type=Connector,Name="HipRoot"), SkipSocket=False) {is=struct RigVMUnitNode}.Result) {variable=Connector Hip Root}
   > RigVMVariableNode(Value=&RigUnit_ResolveArrayConnector(Connector=(Type=Connector,Name="Orient Spaces"), SkipSocket=False).Result) {variable=Tail Spaces}
   > RigVMVariableNode(Value=&RigUnit_ResolveConnector(Connector=(Type=Connector,Name="Parent"), SkipSocket=False).Result) {variable=Connector Parent}
  ]}
 Multiply_1_ContainedGraph {nodes [
   n1 = RigVMFunctionEntryNode()
   Multiply = &RigVMFunction_MathDoubleMul(A=@n1.A, B=@n1.B)
   RigVMFunctionReturnNode(Result=&RigVMFunction_MathDoubleMul(B=@n1.D, A=&RigVMFunction_MathDoubleMul(B=@n1.C, A=@Multiply.Result).Result).Result)
  ]}
 New Function_ContainedGraph {nodes [
   RigVMFunctionEntryNode()
   > RigVMFunction_Sequence_1 = RigVMAggregateNode()
   RigVMFunctionReturnNode(ExecuteContext=@RigVMFunction_Sequence_1.C)
   RigVMDispatchNode(Array=RigVMVariableNode() {variable=Fk Controls}.Value, ExecuteContext=@RigVMFunction_Sequence_1.B) {template="DISPATCH_RigVMDispatch_ArrayReset(io Array)"}
   > RigVMVariableNode(Value=RigVMVariableNode() {variable=Tail Gimbal Control}.Value) {variable=Current Parent Item}
   > For_Each = RigVMDispatchNode(Array=RigVMVariableNode() {variable=Virtual Fk Joints}.Value) {template="DISPATCH_RigVMDispatch_ArrayIterator(in Array,out Element,out Index,out Count,out Ratio)"}
   > RigVMVariableNode(Value=@For_Each.Element) {variable=Current Item}
   > RigVMVariableNode(Value=@For_Each.Index) {variable=Local Index}
   > RigVMVariableNode(Value=@For_Each.Count) {variable=Count}
   Add = /Script/RigVM.RigVMFunction_MathIntAdd(B=1, A=RigVMVariableNode() {variable=Local Index}.Value) {is=struct RigVMUnitNode, template="Add::Execute(in A,in B,out Result)"}
   RigVMFunction_StringPadInteger = /Script/RigVM.RigVMFunction_StringPadInteger(Digits=2, Value=@Add.Result) {is=struct RigVMUnitNode}
   RigVMFunction_StringJoin = /Script/RigVM.RigVMFunction_StringJoin(Separator=" ", Values=("",""), Values.1=@RigVMFunction_StringPadInteger.Result, Values.0=RigVMVariableNode() {variable=Base Name}.Value) {is=struct RigVMUnitNode}
   > RigVMVariableNode(Value=@RigVMFunction_StringJoin.Result) {variable=Token}
   From_String_1 = RigVMDispatchNode(String=RigVMVariableNode() {variable=Token}.Value) {template="DISPATCH_RigDispatch_FromString(in String,out Result)"}
   n1 = RigVMVariableNode() {variable=Local Index}
   n2 = &RigUnit_GetTransform(
    bInitial=true
    Item=RigVMDispatchNode(Array=RigVMVariableNode() {variable=Tail Chain Joints}.Value, Index=@n1.Value) {template="DISPATCH_RigVMDispatch_ArrayGetAtIndex(in Array,in Index,out Element)"}.Element
   )
   Get Scale Factor Profile Value In For Loop = RigVMFunctionReferenceNode(Index=RigVMVariableNode() {variable=Local Index}.Value, Count=RigVMVariableNode() {variable=Count}.Value)
   Prepare Control Shape = RigVMFunctionReferenceNode(
    Scale Premultiplier=1
    Scale Factor=@"Get Scale Factor Profile Value In For Loop".Result
    Control Shape=RigVMVariableNode() {variable=Tail Control Shape}.Value
   )
   Interpolate = /Script/RigVM.RigVMFunction_MathColorLerp(B=(R=0,G=0,B=0,A=1), T=0.7, A=@"Prepare Control Shape".Result Control Shape.Color) {is=struct RigVMUnitNode, template="Interpolate::Execute(in A,in B,in T,out Result)"}
   Modulo = /Script/RigVM.RigVMFunction_MathIntMod(B=2, A=@n1.Value) {is=struct RigVMUnitNode, template="Modulo::Execute(in A,in B,out Result)"}
   Not_Equals = RigVMDispatchNode(B=0, A=@Modulo.Result) {template="DISPATCH_RigVMDispatch_CoreNotEquals(in A,in B,out Result)"}
   RigVMFunction_MathQuaternionFromRotatorV2 = /Script/RigVM.RigVMFunction_MathQuaternionFromRotatorV2(Value=RigVMVariableNode() {variable=FK Control Rotation Offset}.Value) {is=struct RigVMUnitNode, template="Cast::Execute(in Value,out Result)"}
   Multiply = /Script/RigVM.RigVMFunction_MathQuaternionMul(A=@n2.Transform.Rotation, B=@RigVMFunction_MathQuaternionFromRotatorV2.Result) {is=struct RigVMUnitNode, template="Multiply::Execute(in A,in B,out Result)"}
   > SpawnControl = /Script/ControlRig.RigUnit_HierarchyAddControlTransform(
    InitialValue=(Rotation=(X=0,Y=0,Z=0,W=1),Translation=(X=0,Y=0,Z=0),Scale3D=(X=1,Y=1,Z=1))
    OffsetSpace=GlobalSpace
    OffsetTransform=(Rotation=(X=0,Y=0,Z=0,W=1),Translation=(X=0,Y=0,Z=0),Scale3D=(X=1,Y=1,Z=1))
    Settings="(InitialSpace=LocalSpace,bUsePreferredRotationOrder=true,PreferredRotationOrder=XZY,Limits=(LimitTranslationX=(bMinimum=False,bMaximum=False),LimitTranslationY=(bMinimum=False,bMaximum=False),LimitTranslationZ=(bMinimum=False,bMaximum=False),LimitPitch=(bM ...(truncated)"
    Name=@From_String_1.Result
    OffsetTransform.Scale3D=@n2.Transform.Scale3D
    OffsetTransform.Translation=@n2.Transform.Translation
    Settings.Shape.Color=RigVMDispatchNode(False=@Interpolate.Result, Condition=@Not_Equals.Result, True=@"Prepare Control Shape".Result Control Shape.Color) {template="DISPATCH_RigVMDispatch_If(in Condition,in True,in False,out Result)"}.Result
    OffsetTransform.Rotation=@Multiply.Result
    Parent=RigVMVariableNode() {variable=Current Parent Item}.Value
    Settings.PreferredRotationOrder=RigVMVariableNode() {variable=Rotation Order}.Value
    Settings.Shape.bVisible=@"Prepare Control Shape".Result Control Shape.bVisible
    Settings.Shape.Name=@"Prepare Control Shape".Result Control Shape.Name
    Settings.Shape.Transform=@"Prepare Control Shape".Result Control Shape.Transform
   ) {
    is=struct RigVMUnitNode
    template="SpawnControl::Execute(in InitialValue,in Settings,in OffsetTransform,in OffsetSpace,in Parent,in Name,out Item)"
   }
   > RigVMVariableNode(Value=@SpawnControl.Item) {variable=Current Parent Item}
   > RigVMDispatchNode(Array=RigVMVariableNode() {variable=Fk Controls}.Value, Element=@SpawnControl.Item) {template="DISPATCH_RigVMDispatch_ArrayAdd(io Array,in Element,out Index)"}
   RigVMVariableNode(Value=RigVMVariableNode() {variable=Tail Gimbal Control}.Value, ExecuteContext=@RigVMFunction_Sequence_1.A) {variable=Current Parent Item}
   RigVMVariableNode() {variable=Connector Parent}
  ]}
 New Function_ContainedGraph {nodes [
   n1 = RigVMFunctionEntryNode()
   RigVMFunction_Sequence = &RigVMFunction_Sequence(ExecuteContext=@n1.ExecuteContext)
   n2 = &RigVMFunction_Sequence(ExecuteContext=@RigVMFunction_Sequence.A)
   For_Each = RigVMDispatchNode(Array=RigVMVariableNode() {variable=Fk Offset Controls}.Value, ExecuteContext=@n2.A) {template="DISPATCH_RigVMDispatch_ArrayIterator(in Array,out Element,out Index,out Count,out Ratio)"}
   n3 = RigVMDispatchNode(Array=RigVMVariableNode() {variable=Virtual Fk Joints}.Value, Index=@For_Each.Index) {template="DISPATCH_RigVMDispatch_ArrayGetAtIndex(in Array,in Index,out Element)"}
   > &RigUnit_SetTransform(
    Weight=1
    bInitial=False
    Item=@n3.Element
    Value=/Script/ControlRig.RigUnit_ProjectTransformToNewParent(
     bChildInitial=True
     bNewParentInitial=False
     bOldParentInitial=True
     Child=@n3.Element
     NewParent=@For_Each.Element
     OldParent=@For_Each.Element
    ) {is=struct RigVMUnitNode}.Transform
   )
   For_Each_1 = RigVMDispatchNode(Array=RigVMVariableNode() {variable=Virtual Fk Joints}.Value, ExecuteContext=@n2.B) {template="DISPATCH_RigVMDispatch_ArrayIterator(in Array,out Element,out Index,out Count,out Ratio)"}
   n4 = RigVMDispatchNode(Array=RigVMVariableNode() {variable=Tail Chain Joints}.Value, Index=@For_Each_1.Index) {template="DISPATCH_RigVMDispatch_ArrayGetAtIndex(in Array,in Index,out Element)"}
   > &RigUnit_SetTransform(
    Weight=1
    bInitial=False
    Item=@n4.Element
    Value=&RigUnit_ProjectTransformToNewParent(
     bChildInitial=True
     bNewParentInitial=False
     bOldParentInitial=True
     Child=@n4.Element
     NewParent=@For_Each_1.Element
     OldParent=@For_Each_1.Element
    ).Transform
   )
   RigVMFunctionReturnNode(ExecuteContext=@RigVMFunction_Sequence.B)
   For_Each_2 = RigVMDispatchNode(ExecuteContext=(), Array=RigVMVariableNode() {variable=Tail Chain Joints}.Value) {template="DISPATCH_RigVMDispatch_ArrayIterator(in Array,out Element,out Index,out Count,out Ratio)"}
   RigUnit_GetTransform = &RigUnit_GetTransform(
    bInitial=False
    Item=RigVMDispatchNode(Index=@For_Each_2.Index, Array=RigVMVariableNode() {variable=Fk Offset Controls}.Value) {template="DISPATCH_RigVMDispatch_ArrayGetAtIndex(in Array,in Index,out Element)"}.Element
   )
   > &RigUnit_SetTransform(Weight=1, bInitial=False, Item=@For_Each_2.Element, Value=@RigUnit_GetTransform.Transform)
  ]}
 New Function_ContainedGraph {nodes [
   n1 = RigVMFunctionEntryNode()
   RigVMFunction_Sequence = &RigVMFunction_Sequence(ExecuteContext=@n1.ExecuteContext)
   RigVMFunctionReturnNode(ExecuteContext=@RigVMFunction_Sequence.B)
   n2 = &RigVMFunction_ControlFlowBranch(
    Condition=RigVMDispatchNode(Element=RigVMVariableNode() {variable=Tail Rotate ProxyControl}.Value, Array=RigVMVariableNode() {variable=Interacting Items}.Value) {template="DISPATCH_RigVMDispatch_ArrayFind(in Array,in Element,out Index,out Success)"}.Success
    ExecuteContext=@RigVMFunction_Sequence.A
   )
   RigVMFunction_Sequence_1 = RigVMAggregateNode(ExecuteContext=@n2.True)
   RigVMVariableNode(Value=RigVMVariableNode() {variable=Proxy Euler Rotation}.Value, ExecuteContext=@RigVMFunction_Sequence_1.D) {variable=Previous Euler Rotation}
   RigVMVariableNode(ExecuteContext=@RigVMFunction_Sequence_1.C) {variable=Local Count}
   > For_Each = RigVMDispatchNode(Array=RigVMVariableNode() {variable=Fk Controls}.Value) {template="DISPATCH_RigVMDispatch_ArrayIterator(in Array,out Element,out Index,out Count,out Ratio)"}
   > n3 = RigVMVariableNode(Value=@For_Each.Element) {variable=Fk Control}
   GetAnimationChannelFromItem = /Script/ControlRig.RigUnit_GetIntAnimationChannelFromItem(bInitial=False, Item=RigVMVariableNode() {variable=Tail Rotate Position ProxyChannel}.Value) {is=struct RigVMUnitNode, template="GetAnimationChannelFromItem::Execute(out Value,in Item,in bInitial)"}
   Subtract = /Script/RigVM.RigVMFunction_MathIntSub(B=1, A=@GetAnimationChannelFromItem.Value) {is=struct RigVMUnitNode, template="Subtract::Execute(in A,in B,out Result)"}
   GreaterEqual = /Script/RigVM.RigVMFunction_MathIntGreaterEqual(A=@For_Each.Index, B=@Subtract.Result) {is=struct RigVMUnitNode, template="GreaterEqual::Execute(in A,in B,out Result)"}
   n4 = &RigVMFunction_ControlFlowBranch(Condition=@GreaterEqual.Result, ExecuteContext=@n3.ExecuteContext)
   Less = /Script/RigVM.RigVMFunction_MathIntLess(
    B=&RigUnit_GetIntAnimationChannelFromItem(bInitial=False, Item=RigVMVariableNode() {variable=Tail Rotate Spread ProxyChannel}.Value) {template="GetAnimationChannelFromItem::Execute(out Value,in Item,in bInitial)"}.Value
    A=RigVMVariableNode() {variable=Local Count}.Value
   ) {is=struct RigVMUnitNode, template="Less::Execute(in A,in B,out Result)"}
   n5 = &RigVMFunction_ControlFlowBranch(Condition=@Less.Result, ExecuteContext=@n4.True)
   RigVMFunction_MathQuaternionToEuler = &RigVMFunction_MathQuaternionToEuler(
    RotationOrder=ZYX
    Value=&RigUnit_GetTransform(Space=LocalSpace, bInitial=False, Item=RigVMVariableNode() {variable=Fk Control}.Value).Transform.Rotation
   )
   Add = /Script/RigVM.RigVMFunction_MathVectorAdd(A=@RigVMFunction_MathQuaternionToEuler.Result, B=RigVMVariableNode() {variable=Proxy Euler Rotation Delta}.Value) {is=struct RigVMUnitNode, template="Add::Execute(in A,in B,out Result)"}
   RigVMFunction_MathQuaternionFromEuler = &RigVMFunction_MathQuaternionFromEuler(Euler=@Add.Result)
   Set Transform = &RigUnit_SetRotation(
    Weight=1
    bInitial=False
    ExecutePin=@n5.True
    Value=@RigVMFunction_MathQuaternionFromEuler.Result
    Item=RigVMVariableNode() {variable=Fk Control}.Value
   )
   > RigVMVariableNode(Value=&RigVMFunction_MathIntAdd(B=1, A=RigVMVariableNode() {variable=Local Count}.Value) {template="Add::Execute(in A,in B,out Result)"}.Result) {variable=Local Count}
   n6 = RigVMVariableNode() {variable=Tail Rotate ProxyControl}
   RigVMVariableNode(
    Value=&RigVMFunction_MathQuaternionToEuler(RotationOrder=ZYX, Value=&RigUnit_GetTransform(Space=LocalSpace, bInitial=False, Item=@n6.Value).Transform.Rotation).Result
    ExecuteContext=@RigVMFunction_Sequence_1.A
   ) {variable=Proxy Euler Rotation}
   RigVMVariableNode(
    Value=/Script/RigVM.RigVMFunction_MathVectorSub(B=RigVMVariableNode() {variable=Previous Euler Rotation}.Value, A=RigVMVariableNode() {variable=Proxy Euler Rotation}.Value) {is=struct RigVMUnitNode, template="Subtract::Execute(in A,in B,out Result)"}.Result
    ExecuteContext=@RigVMFunction_Sequence_1.B
   ) {variable=Proxy Euler Rotation Delta}
  ]}
 New Function_ContainedGraph {nodes [
   RigVMFunctionEntryNode()
   > RigVMFunction_Sequence = RigVMAggregateNode()
   RigVMDispatchNode(ExecuteContext=@RigVMFunction_Sequence.B, Array=RigVMVariableNode() {variable=Virtual Fk Joints}.Value) {template="DISPATCH_RigVMDispatch_ArrayReset(io Array)"}
   > For_Each = RigVMDispatchNode(Array=RigVMVariableNode() {variable=Tail Chain Joints}.Value) {template="DISPATCH_RigVMDispatch_ArrayIterator(in Array,out Element,out Index,out Count,out Ratio)"}
   > RigVMVariableNode(Value=@For_Each.Element) {variable=Current Item}
   > n1 = RigVMVariableNode(Value=@For_Each.Index) {variable=Local Index}
   Greater = /Script/RigVM.RigVMFunction_MathIntGreater(B=0, A=RigVMVariableNode() {variable=Local Index}.Value) {is=struct RigVMUnitNode, template="Greater::Execute(in A,in B,out Result)"}
   RigVMFunction_ControlFlowBranch = &RigVMFunction_ControlFlowBranch(Condition=@Greater.Result, ExecuteContext=@n1.ExecuteContext)
   n2 = &RigVMFunction_Sequence(ExecuteContext=@RigVMFunction_ControlFlowBranch.True)
   Subtract = &RigVMFunction_MathIntSub(B=1, A=RigVMVariableNode() {variable=Local Index}.Value) {template="Subtract::Execute(in A,in B,out Result)"}
   RigVMVariableNode(
    Value=RigVMDispatchNode(Index=@Subtract.Result, Array=RigVMVariableNode() {variable=Tail Chain Joints}.Value) {template="DISPATCH_RigVMDispatch_ArrayGetAtIndex(in Array,in Index,out Element)"}.Element
    ExecuteContext=@n2.A
   ) {variable=Previous Item}
   n3 = /Script/ControlRig.RigUnit_AimBoneMath(
    DebugSettings=(bEnabled=False,Scale=10,WorldOffset=(Rotation=(X=0,Y=0,Z=0,W=1),Translation=(X=0,Y=0,Z=0),Scale3D=(X=1,Y=1,Z=1)))
    Primary=(Weight=1,Axis=(X=1,Y=0,Z=0),Target=(X=1,Y=0,Z=0),Kind=Location,Space=(Type=Bone,Name="None"))
    Secondary=(Weight=1,Axis=(X=0,Y=0,Z=1),Target=(X=0,Y=0,Z=1),Kind=Direction,Space=(Type=Bone,Name="None"))
    Weight=1
    InputTransform=&RigUnit_GetTransform(bInitial=true, Item=RigVMVariableNode() {variable=Previous Item}.Value).Transform
    Primary.Target=&RigUnit_GetTransform(bInitial=true, Item=RigVMVariableNode() {variable=Current Item}.Value).Transform.Translation
   ) {is=struct RigVMUnitNode}
   To_String = RigVMDispatchNode(Value=RigVMVariableNode() {variable=Current Item}.Value.Name) {template="DISPATCH_RigDispatch_ToString(in Value,out Result)"}
   RigVMFunction_StringJoin = &RigVMFunction_StringJoin(Separator=" ", Values=("Virtual",""), Values.1=@To_String.Result)
   From_String = RigVMDispatchNode(String=@RigVMFunction_StringJoin.Result) {template="DISPATCH_RigDispatch_FromString(in String,out Result)"}
   n4 = /Script/ControlRig.RigUnit_HierarchyAddBone(
    Space=GlobalSpace
    Transform=@n3.Result
    Name=@From_String.Result
    Parent=RigVMVariableNode() {variable=Current Parent Item}.Value
    ExecutePin=@n2.B
   ) {is=struct RigVMUnitNode}
   > RigVMVariableNode(Value=@n4.Item) {variable=Current Parent Item}
   > RigVMDispatchNode(Array=RigVMVariableNode() {variable=Virtual Fk Joints}.Value, Element=@n4.Item) {template="DISPATCH_RigVMDispatch_ArrayAdd(io Array,in Element,out Index)"}
   Add = &RigVMFunction_MathIntAdd(B=1, A=RigVMVariableNode() {variable=Local Index}.Value) {template="Add::Execute(in A,in B,out Result)"}
   From_String_1 = RigVMDispatchNode(String=&RigVMFunction_StringJoin(Separator=" ", Values=("Virtual Tail",""), Values.1=&RigVMFunction_StringPadInteger(Digits=2, Value=@Add.Result).Result).Result) {template="DISPATCH_RigDispatch_FromString(in String,out Result)"}
   n5 = &RigUnit_HierarchyAddBone(
    Space=GlobalSpace
    ExecutePin=@For_Each.Completed
    Name=@From_String_1.Result
    Transform=&RigUnit_GetTransform(bInitial=true, Item=RigVMVariableNode() {variable=Current Item}.Value).Transform
    Parent=RigVMVariableNode() {variable=Current Parent Item}.Value
   )
   > RigVMDispatchNode(Array=RigVMVariableNode() {variable=Virtual Fk Joints}.Value, Element=@n5.Item) {template="DISPATCH_RigVMDispatch_ArrayAdd(io Array,in Element,out Index)"}
   > Set Transform = &RigUnit_SetRotation(Value=(X=0,Y=0,Z=0,W=1), Weight=1, bInitial=true, Item=@n5.Item)
   RigVMFunctionReturnNode(ExecuteContext=@RigVMFunction_Sequence.C)
   RigVMVariableNode(Value=RigVMVariableNode() {variable=Connector Hip Root}.Value, ExecuteContext=@RigVMFunction_Sequence.A) {variable=Current Parent Item}
  ]}
 New Function_ContainedGraph {nodes [
   n1 = RigVMFunctionEntryNode()
   RigVMFunction_Sequence = &RigVMFunction_Sequence(ExecuteContext=@n1.ExecuteContext)
   RigVMFunction_Sequence_1 = RigVMAggregateNode(ExecuteContext=@RigVMFunction_Sequence.A)
   n2 = RigVMDispatchNode(Index=0, Array=RigVMVariableNode() {variable=Fk Controls}.Value) {template="DISPATCH_RigVMDispatch_ArrayGetAtIndex(in Array,in Index,out Element)"}
   Prepare Control Shape = RigVMFunctionReferenceNode(Scale Factor=1, Scale Premultiplier=0.3, Control Shape=RigVMVariableNode() {variable=Tail Rotate Control Shape}.Value)
   n3 = &RigUnit_HierarchyAddControlTransform(
    InitialValue=(Rotation=(X=0,Y=0,Z=0,W=1),Translation=(X=0,Y=0,Z=0),Scale3D=(X=1,Y=1,Z=1))
    Name=Tail Rotate Proxy
    OffsetSpace=GlobalSpace
    Settings="(InitialSpace=LocalSpace,bUsePreferredRotationOrder=true,PreferredRotationOrder=YZX,Limits=(LimitTranslationX=(bMinimum=False,bMaximum=False),LimitTranslationY=(bMinimum=False,bMaximum=False),LimitTranslationZ=(bMinimum=False,bMaximum=False),LimitPitch=(bM ...(truncated)"
    OffsetTransform=&RigUnit_GetTransform(bInitial=true, Item=@n2.Element).Transform
    Parent=/Script/ControlRig.RigUnit_HierarchyGetParent(bDefaultParent=True, Child=@n2.Element) {is=struct RigVMUnitNode}.Parent
    Settings.Shape=@"Prepare Control Shape".Result Control Shape
    Settings.PreferredRotationOrder=RigVMVariableNode() {variable=Rotation Order}.Value
    Settings.Proxy.DrivenControls=RigVMVariableNode() {variable=Fk Controls}.Value
    ExecutePin=@RigVMFunction_Sequence_1.A
   ) {template="SpawnControl::Execute(in InitialValue,in Settings,in OffsetTransform,in OffsetSpace,in Parent,in Name,out Item)"}
   > RigVMVariableNode(Value=@n3.Item) {variable=Tail Rotate ProxyControl}
   Num = RigVMDispatchNode(Array=RigVMVariableNode() {variable=Virtual Fk Joints}.Value) {template="DISPATCH_RigVMDispatch_ArrayGetNum(in Array,out Num)"}
   n4 = /Script/ControlRig.RigUnit_HierarchyAddAnimationChannelInteger(
    InitialValue=1
    LimitsEnabled=(Enabled=(bMinimum=True,bMaximum=true))
    MinimumValue=1
    Name=Position
    MaximumValue=@Num.Num
    Parent=RigVMVariableNode() {variable=Tail Rotate ProxyControl}.Value
    ExecutePin=@RigVMFunction_Sequence_1.B
   ) {is=struct RigVMUnitNode}
   > RigVMVariableNode(Value=@n4.Item) {variable=Tail Rotate Position ProxyChannel}
   n5 = RigVMDispatchNode(Array=RigVMVariableNode() {variable=Virtual Fk Joints}.Value) {template="DISPATCH_RigVMDispatch_ArrayGetNum(in Array,out Num)"}
   > n6 = &RigUnit_HierarchyAddAnimationChannelInteger(
    LimitsEnabled=(Enabled=(bMinimum=True,bMaximum=true))
    MinimumValue=1
    Name=Spread
    InitialValue=@n5.Num
    MaximumValue=@n5.Num
    Parent=RigVMVariableNode() {variable=Tail Rotate ProxyControl}.Value
   )
   > RigVMVariableNode(Value=@n6.Item) {variable=Tail Rotate Spread ProxyChannel}
   SpawnAnimationChannel = /Script/ControlRig.RigUnit_HierarchyAddAnimationChannelBool(
    InitialValue=false
    LimitsEnabled=()
    MaximumValue=True
    MinimumValue=False
    Name=Feedback On Off
    Parent=RigVMVariableNode() {variable=Tail Rotate ProxyControl}.Value
    ExecutePin=@RigVMFunction_Sequence_1.C
   ) {
    is=struct RigVMUnitNode
    template="SpawnAnimationChannel::Execute(in InitialValue,in MinimumValue,in MaximumValue,in LimitsEnabled,in Parent,in Name,out Item)"
   }
   > RigVMVariableNode(Value=@SpawnAnimationChannel.Item) {variable=Tail Rotate Feedback ProxyChannel}
   > n7 = &RigUnit_HierarchyAddAnimationChannelInteger(
    InitialValue=5
    LimitsEnabled=(Enabled=(bMinimum=True,bMaximum=true))
    MaximumValue=20
    MinimumValue=1
    Name=Feedback Thickness
    Parent=RigVMVariableNode() {variable=Tail Rotate ProxyControl}.Value
   )
   > RigVMVariableNode(Value=@n7.Item) {variable=Tail Rotate FeedbackThickness ProxyChannel}
   RigVMFunctionReturnNode(ExecuteContext=@RigVMFunction_Sequence.B)
  ]}
 New Function_ContainedGraph {nodes [
   n1 = RigVMFunctionEntryNode()
   RigVMFunction_Sequence = &RigVMFunction_Sequence(ExecuteContext=@n1.ExecuteContext)
   RigVMFunctionReturnNode(ExecuteContext=@RigVMFunction_Sequence.B)
   n2 = &RigVMFunction_MathIntSub(
    B=1
    A=&RigUnit_GetIntAnimationChannelFromItem(bInitial=False, Item=RigVMVariableNode() {variable=Tail Rotate Position ProxyChannel}.Value) {template="GetAnimationChannelFromItem::Execute(out Value,in Item,in bInitial)"}.Value
   ) {template="Subtract::Execute(in A,in B,out Result)"}
   n3 = RigVMVariableNode() {variable=Fk Controls}
   RigUnit_GetTransform = &RigUnit_GetTransform(
    bInitial=False
    Item=&RigUnit_HierarchyGetParent(
     bDefaultParent=True
     Child=RigVMDispatchNode(Index=@n2.Result, Array=@n3.Value) {template="DISPATCH_RigVMDispatch_ArrayGetAtIndex(in Array,in Index,out Element)"}.Element
    ).Parent
   )
   RigUnit_SetTransform = &RigUnit_SetTransform(
    Weight=1
    bInitial=False
    Value=@RigUnit_GetTransform.Transform
    Item=RigVMVariableNode() {variable=Tail Rotate Proxy Control Parent}.Value
    ExecutePin=@RigVMFunction_Sequence.A
   )
  ]}
 New Function_ContainedGraph {nodes [
   RigVMFunctionEntryNode()
   > RigVMFunction_Sequence = RigVMAggregateNode()
   n1 = /Script/ControlRig.RigUnit_IsInteracting() {is=struct RigVMUnitNode}
   RigVMVariableNode(Value=@n1.Items, ExecuteContext=@RigVMFunction_Sequence.C) {variable=Interacting Items}
   > RigVMVariableNode(Value=@n1.bIsRotating) {variable=Is Interacting}
   RigVMFunctionReturnNode(ExecuteContext=@RigVMFunction_Sequence.D)
   n2 = &RigVMFunction_ControlFlowBranch(Condition=@n1.bIsRotating, ExecuteContext=@RigVMFunction_Sequence.A)
   RigVMVariableNode(Value=(X=0,Y=0,Z=0), ExecuteContext=@n2.False) {variable=Previous Euler Rotation}
   > &RigUnit_SetTransform(
    Space=LocalSpace
    Value=(Rotation=(X=0,Y=0,Z=0,W=1),Translation=(X=0,Y=0,Z=0),Scale3D=(X=1,Y=1,Z=1))
    Weight=1
    bInitial=False
    Item=RigVMVariableNode() {variable=Tail Rotate ProxyControl}.Value
   )
   Proxy Interaction = RigVMFunctionReferenceNode(ExecuteContext=@n2.True)
  ]}
 New Function_ContainedGraph {nodes [
   n1 = RigVMFunctionEntryNode()
   n2 = &RigVMFunction_Sequence(ExecuteContext=@n1.ExecuteContext)
   RigVMFunction_Sequence = RigVMAggregateNode(ExecuteContext=@n2.A)
   RigVMDispatchNode(Array=RigVMVariableNode() {variable=Curve Positions}.Value, ExecuteContext=@RigVMFunction_Sequence.A) {template="DISPATCH_RigVMDispatch_ArrayReset(io Array)"}
   > For_Each = RigVMDispatchNode(Array=RigVMVariableNode() {variable=Virtual Fk Joints}.Value) {template="DISPATCH_RigVMDispatch_ArrayIterator(in Array,out Element,out Index,out Count,out Ratio)"}
   > RigVMVariableNode(Value=@For_Each.Element) {variable=Local Item}
   > n3 = RigVMVariableNode(Value=@For_Each.Index) {variable=Local Index}
   GetAnimationChannelFromItem = &RigUnit_GetIntAnimationChannelFromItem(bInitial=False, Item=RigVMVariableNode() {variable=Tail Rotate Position ProxyChannel}.Value) {template="GetAnimationChannelFromItem::Execute(out Value,in Item,in bInitial)"}
   Subtract = &RigVMFunction_MathIntSub(B=1, A=@GetAnimationChannelFromItem.Value) {template="Subtract::Execute(in A,in B,out Result)"}
   GreaterEqual = &RigVMFunction_MathIntGreaterEqual(B=@Subtract.Result, A=RigVMVariableNode() {variable=Local Index}.Value) {template="GreaterEqual::Execute(in A,in B,out Result)"}
   RigVMFunction_ControlFlowBranch = &RigVMFunction_ControlFlowBranch(Condition=@GreaterEqual.Result, ExecuteContext=@n3.ExecuteContext)
   LessEqual = /Script/RigVM.RigVMFunction_MathIntLessEqual(
    B=&RigUnit_GetIntAnimationChannelFromItem(bInitial=False, Item=RigVMVariableNode() {variable=Tail Rotate Spread ProxyChannel}.Value) {template="GetAnimationChannelFromItem::Execute(out Value,in Item,in bInitial)"}.Value
    A=RigVMVariableNode() {variable=Local Count}.Value
   ) {is=struct RigVMUnitNode, template="LessEqual::Execute(in A,in B,out Result)"}
   n4 = &RigVMFunction_ControlFlowBranch(Condition=@LessEqual.Result, ExecuteContext=@RigVMFunction_ControlFlowBranch.True)
   Add = RigVMDispatchNode(
    Element=&RigUnit_GetTransform(bInitial=False, Item=RigVMVariableNode() {variable=Local Item}.Value).Transform.Translation
    ExecuteContext=@n4.True
    Array=RigVMVariableNode() {variable=Curve Positions}.Value
   ) {template="DISPATCH_RigVMDispatch_ArrayAdd(io Array,in Element,out Index)"}
   > RigVMVariableNode(Value=&RigVMFunction_MathIntAdd(B=1, A=RigVMVariableNode() {variable=Local Count}.Value) {template="Add::Execute(in A,in B,out Result)"}.Result) {variable=Local Count}
   RigVMFunction_MathIntToFloat = /Script/RigVM.RigVMFunction_MathIntToFloat(Value=&RigUnit_GetIntAnimationChannelFromItem(bInitial=False, Item=RigVMVariableNode() {variable=Tail Rotate FeedbackThickness ProxyChannel}.Value) {template="GetAnimationChannelFromItem::Execute(out Value,in Item,in bInitial)"}.Value) {is=struct RigVMUnitNode, template="Cast::Execute(in Value,out Result)"}
   RigVMFunction_DebugLineStripNoSpace = /Script/RigVM.RigVMFunction_DebugLineStripNoSpace(
    Color=(R=1,G=0,B=0,A=1)
    DebugDrawSettings=(DepthPriority=SDPG_Foreground,Lifetime=-1)
    WorldOffset=(Rotation=(X=0,Y=0,Z=0,W=1),Translation=(X=0,Y=0,Z=0),Scale3D=(X=1,Y=1,Z=1))
    bEnabled=/Script/ControlRig.RigUnit_GetBoolAnimationChannelFromItem(bInitial=False, Item=RigVMVariableNode() {variable=Tail Rotate Feedback ProxyChannel}.Value) {is=struct RigVMUnitNode, template="GetAnimationChannelFromItem::Execute(out Value,in Item,in bInitial)"}.Value
    Thickness=@RigVMFunction_MathIntToFloat.Result
    Points=RigVMVariableNode() {variable=Curve Positions}.Value
    ExecutePin=@RigVMFunction_Sequence.B
   ) {is=struct RigVMUnitNode}
   RigVMFunctionReturnNode(ExecuteContext=@n2.B)
  ]}
 New Function_ContainedGraph {nodes [
   RigVMFunctionEntryNode()
   > RigVMFunction_Sequence = RigVMAggregateNode()
   For_Each = RigVMDispatchNode(Array=RigVMVariableNode() {variable=Tail Chain Joints}.Value, ExecuteContext=@RigVMFunction_Sequence.B) {template="DISPATCH_RigVMDispatch_ArrayIterator(in Array,out Element,out Index,out Count,out Ratio)"}
   n1 = RigVMDispatchNode(Array=RigVMVariableNode() {variable=Fk Controls}.Value, Index=@For_Each.Index) {template="DISPATCH_RigVMDispatch_ArrayGetAtIndex(in Array,in Index,out Element)"}
   > &RigUnit_SetTransform(
    Weight=1
    bInitial=False
    Value=&RigUnit_ProjectTransformToNewParent(
     bChildInitial=True
     bNewParentInitial=False
     bOldParentInitial=True
     Child=@n1.Element
     NewParent=@For_Each.Element
     OldParent=@For_Each.Element
    ).Transform
    Item=@n1.Element
   )
   n2 = RigVMVariableNode() {variable=Tail Gimbal Control}
   n3 = RigVMVariableNode() {variable=Connector Tail Base}
   &RigUnit_SetTransform(
    Weight=1
    bInitial=False
    Value=&RigUnit_ProjectTransformToNewParent(bChildInitial=True, bNewParentInitial=False, bOldParentInitial=True, Child=@n2.Value, NewParent=@n3.Value, OldParent=@n3.Value).Transform
    ExecutePin=@RigVMFunction_Sequence.A
    Item=@n2.Value
   )
   RigVMFunctionReturnNode(ExecuteContext=@RigVMFunction_Sequence.C)
  ]}
 New Function_ContainedGraph {nodes [
   n1 = RigVMFunctionEntryNode()
   RigVMFunction_Sequence = &RigVMFunction_Sequence(ExecuteContext=@n1.ExecuteContext)
   RigVMFunction_Sequence_1 = RigVMAggregateNode(ExecuteContext=@RigVMFunction_Sequence.A)
   RigVMDispatchNode(Array=RigVMVariableNode() {variable=Fk Offset Controls}.Value, ExecuteContext=@RigVMFunction_Sequence_1.A) {template="DISPATCH_RigVMDispatch_ArrayReset(io Array)"}
   > For_Each = RigVMDispatchNode(Array=RigVMVariableNode() {variable=Fk Controls}.Value) {template="DISPATCH_RigVMDispatch_ArrayIterator(in Array,out Element,out Index,out Count,out Ratio)"}
   > RigVMVariableNode(Value=@For_Each.Index) {variable=Local Index}
   RigVMFunction_StringPadInteger = &RigVMFunction_StringPadInteger(Digits=2, Value=&RigVMFunction_MathIntAdd(B=1, A=@For_Each.Index) {template="Add::Execute(in A,in B,out Result)"}.Result)
   RigVMFunction_StringJoin = &RigVMFunction_StringJoin(Separator=" ", Values=("Tail","","Offset"), Values.1=@RigVMFunction_StringPadInteger.Result)
   From_String = RigVMDispatchNode(String=@RigVMFunction_StringJoin.Result) {template="DISPATCH_RigDispatch_FromString(in String,out Result)"}
   Get Scale Factor Profile Value In For Loop = RigVMFunctionReferenceNode(Index=@For_Each.Index, Count=@For_Each.Count)
   Prepare Control Shape_0 = RigVMFunctionReferenceNode(
    Scale Premultiplier=1
    Scale Factor=@"Get Scale Factor Profile Value In For Loop".Result
    Control Shape=RigVMVariableNode() {variable=Tail Fk Offset Control Shape}.Value
   )
   > SpawnControl = &RigUnit_HierarchyAddControlTransform(
    InitialValue=(Rotation=(X=0,Y=0,Z=0,W=1),Translation=(X=0,Y=0,Z=0),Scale3D=(X=1,Y=1,Z=1))
    OffsetSpace=LocalSpace
    OffsetTransform=(Rotation=(X=0,Y=0,Z=0,W=1),Translation=(X=0,Y=0,Z=0),Scale3D=(X=1,Y=1,Z=1))
    Settings="(InitialSpace=LocalSpace,bUsePreferredRotationOrder=False,PreferredRotationOrder=YZX,Limits=(LimitTranslationX=(bMinimum=False,bMaximum=False),LimitTranslationY=(bMinimum=False,bMaximum=False),LimitTranslationZ=(bMinimum=False,bMaximum=False),LimitPitch=(b ...(truncated)"
    Name=@From_String.Result
    Settings.Shape=@"Prepare Control Shape_0".Result Control Shape
    Settings.PreferredRotationOrder=RigVMVariableNode() {variable=Rotation Order}.Value
    Parent=@For_Each.Element
   ) {template="SpawnControl::Execute(in InitialValue,in Settings,in OffsetTransform,in OffsetSpace,in Parent,in Name,out Item)"}
   > Set_Metadata = RigVMDispatchNode(
    Name=Control
    NameSpace=Self
    Value=RigVMDispatchNode(Array=RigVMVariableNode() {variable=Tail Chain Joints}.Value, Index=RigVMVariableNode() {variable=Local Index}.Value) {template="DISPATCH_RigVMDispatch_ArrayGetAtIndex(in Array,in Index,out Element)"}.Element
    Item=@SpawnControl.Item
   ) {template="DISPATCH_RigDispatch_SetMetadata(in Item,in Name,in NameSpace,in Value,out Success)"}
   > Add = RigVMDispatchNode(Array=RigVMVariableNode() {variable=Fk Offset Controls}.Value, Element=@SpawnControl.Item) {template="DISPATCH_RigVMDispatch_ArrayAdd(io Array,in Element,out Index)"}
   n2 = &RigUnit_HierarchyAddAnimationChannelBool(
    InitialValue=false
    LimitsEnabled=()
    MaximumValue=True
    MinimumValue=false
    Name=Fk Offset Controls Vis
    Parent=RigVMVariableNode() {variable=Tail Gimbal Control}.Value
    ExecutePin=@RigVMFunction_Sequence_1.B
   ) {template="SpawnAnimationChannel::Execute(in InitialValue,in MinimumValue,in MaximumValue,in LimitsEnabled,in Parent,in Name,out Item)"}
   > /Script/ControlRig.RigUnit_SetChannelHosts(Channel=@n2.Item, Hosts=RigVMVariableNode() {variable=Fk Offset Controls}.Value) {is=struct RigVMUnitNode}
   > RigVMVariableNode(Value=@n2.Item) {variable=Fk Offset Controls Visibility}
   RigVMFunctionReturnNode(ExecuteContext=@RigVMFunction_Sequence.B)
  ]}
 New Function_ContainedGraph {nodes [
   RigVMFunctionEntryNode()
   > RigVMFunction_Sequence = RigVMAggregateNode()
   RigVMFunctionReturnNode(ExecuteContext=@RigVMFunction_Sequence.C)
   Construct Space Nulls v01 = RigVMFunctionReferenceNode(
    Orient Only=true
    Space Null Parent=&RigUnit_HierarchyGetParent(bDefaultParent=True, Child=RigVMVariableNode() {variable=Position Parent Null}.Value).Parent
    Spaces=RigVMVariableNode() {variable=Tail Spaces}.Value
    Driven Control=RigVMVariableNode() {variable=Tail Gimbal Control}.Value
    ExecuteContext=@RigVMFunction_Sequence.B
   )
   > RigVMVariableNode(Value=@"Construct Space Nulls v01".Output Space Nulls) {variable=Tail Space Nulls}
   n1 = &RigUnit_HierarchyAddNull(
    Name=Parent Null
    Space=GlobalSpace
    Transform=(Rotation=(X=0,Y=0,Z=0,W=1),Translation=(X=0,Y=0,Z=0),Scale3D=(X=1,Y=1,Z=1))
    Transform.Translation=&RigUnit_GetTransform(bInitial=false, Item=RigVMVariableNode() {variable=Tail Gimbal Control}.Value).Transform.Translation
    Parent=RigVMVariableNode() {variable=Connector Hip Root}.Value
    ExecutePin=@RigVMFunction_Sequence.A
   )
   > RigVMVariableNode(Value=@n1.Item) {variable=Position Parent Null}
  ]}
 New Function_ContainedGraph {nodes [
   RigVMFunctionEntryNode()
   > Evaluate Space Nulls v01 = RigVMFunctionReferenceNode(Orient Only=true, Spaces=RigVMVariableNode() {variable=Tail Spaces}.Value, Space Nulls=RigVMVariableNode() {variable=Tail Space Nulls}.Value)
   > RigVMFunctionReturnNode()
  ]}
 New Function_ContainedGraph {nodes [
   RigVMFunctionEntryNode()
   > RigVMFunction_Sequence = RigVMAggregateNode()
   RigVMFunctionReturnNode(ExecuteContext=@RigVMFunction_Sequence.C)
   n1 = &RigUnit_GetTransform(bInitial=true, Item=RigVMVariableNode() {variable=Connector Tail Base}.Value)
   Multiply = &RigVMFunction_MathQuaternionMul(
    A=@n1.Transform.Rotation
    B=&RigVMFunction_MathQuaternionFromRotatorV2(Value=RigVMVariableNode() {variable=FK Control Rotation Offset}.Value) {template="Cast::Execute(in Value,out Result)"}.Result
   ) {template="Multiply::Execute(in A,in B,out Result)"}
   n2 = &RigUnit_HierarchyAddNull(
    Name=Parent Null
    Space=GlobalSpace
    Transform=(Rotation=(X=0,Y=0,Z=0,W=1),Translation=(X=0,Y=0,Z=0),Scale3D=(X=1,Y=1,Z=1))
    Transform.Translation=@n1.Transform.Translation
    Transform.Rotation=@Multiply.Result
    Parent=RigVMVariableNode() {variable=Current Parent Item}.Value
    ExecutePin=@RigVMFunction_Sequence.B
   )
   > RigVMVariableNode(Value=@n2.Item) {variable=Position Parent Null}
   Prepare Control Shape = RigVMFunctionReferenceNode(Scale Factor=1, Scale Premultiplier=0.3, Control Shape=RigVMVariableNode() {variable=Tail Gimbal Control Shape}.Value)
   > n3 = &RigUnit_HierarchyAddControlTransform(
    InitialValue=(Rotation=(X=0,Y=0,Z=0,W=1),Translation=(X=0,Y=0,Z=0),Scale3D=(X=1,Y=1,Z=1))
    Name=TailBase
    OffsetSpace=LocalSpace
    OffsetTransform=(Rotation=(X=0,Y=0,Z=0,W=1),Translation=(X=0,Y=0,Z=0),Scale3D=(X=1,Y=1,Z=1))
    Settings="(InitialSpace=LocalSpace,bUsePreferredRotationOrder=true,PreferredRotationOrder=YZX,Limits=(LimitTranslationX=(bMinimum=False,bMaximum=False),LimitTranslationY=(bMinimum=False,bMaximum=False),LimitTranslationZ=(bMinimum=False,bMaximum=False),LimitPitch=(bM ...(truncated)"
    Parent=@n2.Item
    Settings.Shape=@"Prepare Control Shape".Result Control Shape
    Settings.PreferredRotationOrder=RigVMVariableNode() {variable=Rotation Order}.Value
    Settings.DisplayName=RigVMVariableNode() {variable=Gimbal Control Name}.Value
   ) {template="SpawnControl::Execute(in InitialValue,in Settings,in OffsetTransform,in OffsetSpace,in Parent,in Name,out Item)"}
   > RigVMVariableNode(Value=@n3.Item) {variable=Tail Gimbal Control}
   RigVMVariableNode(
    Value=&RigUnit_ResolveConnector(Connector=(Type=Connector,Name="Parent"), SkipSocket=False).Result
    ExecuteContext=@RigVMFunction_Sequence.A
   ) {variable=Current Parent Item}
  ]}
 New Function_ContainedGraph {nodes [
   n1 = RigVMFunctionEntryNode()
   RigVMFunction_Sequence = &RigVMFunction_Sequence(ExecuteContext=@n1.ExecuteContext)
   RigVMFunctionReturnNode(ExecuteContext=@RigVMFunction_Sequence.B)
   For_Each = RigVMDispatchNode(Array=RigVMVariableNode() {variable=Fk Offset Controls}.Value, ExecuteContext=@RigVMFunction_Sequence.A) {template="DISPATCH_RigVMDispatch_ArrayIterator(in Array,out Element,out Index,out Count,out Ratio)"}
   GetAnimationChannelFromItem = &RigUnit_GetBoolAnimationChannelFromItem(bInitial=False, Item=RigVMVariableNode() {variable=Fk Offset Controls Visibility}.Value) {template="GetAnimationChannelFromItem::Execute(out Value,in Item,in bInitial)"}
   > SetControlVisibility = /Script/ControlRig.RigUnit_SetControlVisibility(Item=@For_Each.Element, bVisible=@GetAnimationChannelFromItem.Value) {is=struct RigVMUnitNode, template="SetControlVisibility::Execute(in Item,in Pattern,in bVisible)"}
  ]}
 New Function_ContainedGraph {nodes [
   n1 = RigVMFunctionEntryNode()
   Get_Module_Metadata = RigVMDispatchNode(Name=Global Center Control COlor, NameSpace=Root, Default=@n1.Control Shape.Color) {template="DISPATCH_RigDispatch_GetModuleMetadata(in Name,in NameSpace,in Default,out Value,out Found)"}
   Interpolate = &RigVMFunction_MathColorLerp(B=@Get_Module_Metadata.Value, T=RigVMVariableNode() {variable=Root Module Color Weight}.Value, A=@n1.Control Shape.Color) {template="Interpolate::Execute(in A,in B,in T,out Result)"}
   Get_Module_Metadata_3 = RigVMDispatchNode(Default=1, Name=Global Control Scale, NameSpace=Root) {template="DISPATCH_RigDispatch_GetModuleMetadata(in Name,in NameSpace,in Default,out Value,out Found)"}
   Multiply = RigVMAggregateNode(
    D=@Get_Module_Metadata_3.Value
    A=RigVMVariableNode() {variable=Scale Factor}.Value
    C=RigVMVariableNode() {variable=Scale Premultiplier}.Value
    B=RigVMVariableNode() {variable=Control Scale}.Value
   )
   n2 = RigVMVariableNode() {variable=Control Shape}
   Scale = /Script/RigVM.RigVMFunction_MathVectorScale(Factor=@Multiply.Result, Value=@n2.Value.Transform.Scale3D) {is=struct RigVMUnitNode, template="Scale::Execute(in Value,in Factor,out Result)"}
   > RigVMFunctionReturnNode(
    Result Control Shape=(bVisible=True,Name="Default",Color=(R=1,G=0,B=0,A=1),Transform=(Rotation=(X=0,Y=0,Z=0,W=1),Translation=(X=0,Y=0,Z=0),Scale3D=(X=1,Y=1,Z=1)))
    Result Control Shape.Name=@n1.Control Shape.Name
    Result Control Shape.bVisible=@n1.Control Shape.bVisible
    Result Control Shape.Color=@Interpolate.Result
    Result Control Shape.Transform.Scale3D=@Scale.Result
    Result Control Shape.Transform.Rotation=@n2.Value.Transform.Rotation
    Result Control Shape.Transform.Translation=@n2.Value.Transform.Translation
   )
  ]}
 New Function_ContainedGraph {nodes [
   n1 = RigVMFunctionEntryNode()
   Add = &RigVMFunction_MathIntAdd(B=1, A=@n1.Index) {template="Add::Execute(in A,in B,out Result)"}
   Cast = &RigVMFunction_MathIntToFloat(Value=@Add.Result) {template="Cast::Execute(in Value,out Result)"}
   n2 = &RigVMFunction_MathIntToFloat(Value=@n1.Count) {template="Cast::Execute(in Value,out Result)"}
   LessEqual = &RigVMFunction_MathIntLessEqual(B=0, A=@n1.Count) {template="LessEqual::Execute(in A,in B,out Result)"}
   Get Scale Factor Profile Value = RigVMFunctionReferenceNode(Param=RigVMDispatchNode(True=0, False=&RigVMFunction_MathDoubleDiv(A=@Cast.Result, B=@n2.Result).Result, Condition=@LessEqual.Result) {template="DISPATCH_RigVMDispatch_If(in Condition,in True,in False,out Result)"}.Result)
   > RigVMFunctionReturnNode(Result=@"Get Scale Factor Profile Value".Result)
  ]}
 New Function_ContainedGraph {nodes [
   n1 = RigVMFunctionEntryNode()
   > RigVMFunctionReturnNode(Result=&RigVMFunction_AnimEvalRichCurve(SourceMinimum=0, TargetMinimum=0, Value=@n1.Param, Curve=RigVMVariableNode() {variable=Control Scale Factor Profile}.Value).Result)
  ]}
 RigVMFunctionLibrary {nodes [
   Check If Interacting With Controls = RigVMCollapseNode(ExecuteContext=())
   Connect Orient Space Switches = RigVMCollapseNode(ExecuteContext=())
   Create Chain Fk Controls = RigVMCollapseNode(ExecuteContext=())
   Create Chain Fk Joints = RigVMCollapseNode(ExecuteContext=())
   Create Fk Offset Controls = RigVMCollapseNode(ExecuteContext=())
   Create Gimbal Control = RigVMCollapseNode(ExecuteContext=())
   Create Proxy Controls = RigVMCollapseNode(ExecuteContext=())
   Create Space Switches = RigVMCollapseNode(ExecuteContext=())
   Get Scale Factor Profile Value = RigVMCollapseNode(Param=0)
   Get Scale Factor Profile Value In For Loop = RigVMCollapseNode(Count=0, Index=0)
   Prepare Control Shape = RigVMCollapseNode(
    Control Shape=(bVisible=True,Name="Default",Color=(R=1,G=0,B=0,A=1),Transform=(Rotation=(X=0,Y=0,Z=0,W=1),Translation=(X=0,Y=0,Z=0),Scale3D=(X=1,Y=1,Z=1)))
    Scale Factor=1
    Scale Premultiplier=1
   )
   Proxy Feedback = RigVMCollapseNode(ExecuteContext=())
   Proxy Interaction = RigVMCollapseNode(ExecuteContext=())
   Set Controls From Nulls = RigVMCollapseNode(ExecuteContext=())
   Set Joints From Controls = RigVMCollapseNode(ExecuteContext=())
   Set Offset Fk Control Visibility = RigVMCollapseNode(ExecuteContext=())
   Set Proxy Position = RigVMCollapseNode(ExecuteContext=())
  ]}
 RigVMFunction_Sequence_1_1_ContainedGraph {nodes [
   n1 = RigVMFunctionEntryNode()
   n2 = &RigVMFunction_Sequence(ExecuteContext=@n1.ExecuteContext)
   n3 = &RigVMFunction_Sequence(ExecuteContext=@n2.B)
   n4 = &RigVMFunction_Sequence(ExecuteContext=@n3.B)
   n5 = &RigVMFunction_Sequence(ExecuteContext=@n4.B)
   RigVMFunctionReturnNode(A=@n2.A, B=@n3.A, C=@n4.A, D=@n5.A, E=@n5.B)
  ]}
 RigVMFunction_Sequence_1_1_ContainedGraph {nodes [
   n1 = RigVMFunctionEntryNode()
   n2 = &RigVMFunction_Sequence(ExecuteContext=@n1.ExecuteContext)
   n3 = &RigVMFunction_Sequence(ExecuteContext=@n2.B)
   n4 = &RigVMFunction_Sequence(ExecuteContext=@n3.B)
   n5 = &RigVMFunction_Sequence(ExecuteContext=@n4.B)
   RigVMFunctionReturnNode(A=@n2.A, B=@n3.A, C=@n4.A, D=@n5.A, E=@n5.B)
  ]}
 RigVMFunction_Sequence_1_ContainedGraph {nodes [
   n1 = RigVMFunctionEntryNode()
   RigVMFunction_Sequence = &RigVMFunction_Sequence(ExecuteContext=@n1.ExecuteContext)
   n2 = &RigVMFunction_Sequence(ExecuteContext=@RigVMFunction_Sequence.B)
   n3 = &RigVMFunction_Sequence(ExecuteContext=@n2.B)
   n4 = &RigVMFunction_Sequence(ExecuteContext=@n3.B)
   RigVMFunctionReturnNode(A=@RigVMFunction_Sequence.A, B=@n2.A, C=@n3.A, D=@n4.A, E=@n4.B)
  ]}
 RigVMFunction_Sequence_1_ContainedGraph {nodes [
   n1 = RigVMFunctionEntryNode()
   RigVMFunction_Sequence = &RigVMFunction_Sequence(ExecuteContext=@n1.ExecuteContext)
   n2 = &RigVMFunction_Sequence(ExecuteContext=@RigVMFunction_Sequence.B)
   n3 = &RigVMFunction_Sequence(ExecuteContext=@n2.B)
   n4 = &RigVMFunction_Sequence(ExecuteContext=@n3.B)
   RigVMFunctionReturnNode(A=@RigVMFunction_Sequence.A, B=@n2.A, C=@n3.A, D=@n4.A, E=@n4.B)
  ]}
 RigVMFunction_Sequence_1_ContainedGraph {nodes [
   n1 = RigVMFunctionEntryNode()
   RigVMFunction_Sequence = &RigVMFunction_Sequence(ExecuteContext=@n1.ExecuteContext)
   n2 = &RigVMFunction_Sequence(ExecuteContext=@RigVMFunction_Sequence.B)
   n3 = &RigVMFunction_Sequence(ExecuteContext=@n2.B)
   RigVMFunctionReturnNode(A=@RigVMFunction_Sequence.A, B=@n2.A, C=@n3.A, D=@n3.B)
  ]}
 RigVMFunction_Sequence_1_ContainedGraph {nodes [
   n1 = RigVMFunctionEntryNode()
   RigVMFunction_Sequence = &RigVMFunction_Sequence(ExecuteContext=@n1.ExecuteContext)
   n2 = &RigVMFunction_Sequence(ExecuteContext=@RigVMFunction_Sequence.B)
   RigVMFunctionReturnNode(A=@RigVMFunction_Sequence.A, B=@n2.A, C=@n2.B)
  ]}
 RigVMFunction_Sequence_1_ContainedGraph {nodes [
   n1 = RigVMFunctionEntryNode()
   RigVMFunction_Sequence = &RigVMFunction_Sequence(ExecuteContext=@n1.ExecuteContext)
   n2 = &RigVMFunction_Sequence(ExecuteContext=@RigVMFunction_Sequence.B)
   RigVMFunctionReturnNode(A=@RigVMFunction_Sequence.A, B=@n2.A, C=@n2.B)
  ]}
 RigVMFunction_Sequence_1_ContainedGraph {nodes [
   n1 = RigVMFunctionEntryNode()
   RigVMFunction_Sequence = &RigVMFunction_Sequence(ExecuteContext=@n1.ExecuteContext)
   n2 = &RigVMFunction_Sequence(ExecuteContext=@RigVMFunction_Sequence.B)
   RigVMFunctionReturnNode(A=@RigVMFunction_Sequence.A, B=@n2.A, C=@n2.B)
  ]}
 RigVMFunction_Sequence_1_ContainedGraph {nodes [
   n1 = RigVMFunctionEntryNode()
   RigVMFunction_Sequence = &RigVMFunction_Sequence(ExecuteContext=@n1.ExecuteContext)
   n2 = &RigVMFunction_Sequence(ExecuteContext=@RigVMFunction_Sequence.B)
   RigVMFunctionReturnNode(A=@RigVMFunction_Sequence.A, B=@n2.A, C=@n2.B)
  ]}
 RigVMFunction_Sequence_2_ContainedGraph {nodes [
   n1 = RigVMFunctionEntryNode()
   n2 = &RigVMFunction_Sequence(ExecuteContext=@n1.ExecuteContext)
   n3 = &RigVMFunction_Sequence(ExecuteContext=@n2.B)
   RigVMFunctionReturnNode(A=@n2.A, B=@n3.A, C=@n3.B)
  ]}
 RigVMFunction_Sequence_2_ContainedGraph {nodes [
   n1 = RigVMFunctionEntryNode()
   n2 = &RigVMFunction_Sequence(ExecuteContext=@n1.ExecuteContext)
   n3 = &RigVMFunction_Sequence(ExecuteContext=@n2.B)
   RigVMFunctionReturnNode(A=@n2.A, B=@n3.A, C=@n3.B)
  ]}
 RigVMFunction_Sequence_3_ContainedGraph {nodes [
   n1 = RigVMFunctionEntryNode()
   RigVMFunction_Sequence = &RigVMFunction_Sequence(ExecuteContext=@n1.ExecuteContext)
   n2 = &RigVMFunction_Sequence(ExecuteContext=@RigVMFunction_Sequence.B)
   RigVMFunctionReturnNode(A=@RigVMFunction_Sequence.A, B=@n2.A, C=@n2.B)
  ]}
 RigVMModel {nodes [
   RigUnit_BeginExecution = &RigUnit_BeginExecution()
   RigVMFunction_Sequence = RigVMAggregateNode(ExecuteContext=@RigUnit_BeginExecution.ExecutePin)
   Set Proxy Position = RigVMFunctionReferenceNode(ExecuteContext=@RigVMFunction_Sequence.A)
   Check If Interacting With Controls = RigVMFunctionReferenceNode(ExecuteContext=@RigVMFunction_Sequence.B)
   Proxy Feedback = RigVMFunctionReferenceNode(ExecuteContext=@RigVMFunction_Sequence.C)
   Connect Orient Space Switches = RigVMFunctionReferenceNode(ExecuteContext=@RigVMFunction_Sequence.D)
   > Set Offset Fk Control Visibility = RigVMFunctionReferenceNode()
   > Set Joints From Controls = RigVMFunctionReferenceNode()
  ]}
 RigVMModel Backward Solve Graph {nodes [
   n1 = /Script/ControlRig.RigUnit_InverseExecution() {is=struct RigVMUnitNode}
   Set Controls From Nulls = RigVMFunctionReferenceNode(ExecuteContext=@n1.ExecutePin)
  ]}
 RigVMModel Connector Graph {nodes [
   n1 = /Script/ControlRig.RigUnit_ConnectorExecution() {is=struct RigVMUnitNode}
   n2 = &RigUnit_ResolveArrayConnector(Connector=(Type=Connector,Name="Tail"), SkipSocket=False)
   Num = RigVMDispatchNode(Array=@n2.Result) {template="DISPATCH_RigVMDispatch_ArrayGetNum(in Array,out Num)"}
   RigVMFunction_MathBoolAnd = &RigVMFunction_MathBoolAnd(
    A=RigVMDispatchNode(A=(Type=Connector,Name="Tail"), B=/Script/ControlRig.RigUnit_GetCandidates() {is=struct RigVMUnitNode}.Connector) {template="DISPATCH_RigVMDispatch_CoreEquals(in A,in B,out Result)"}.Result
    B=RigVMDispatchNode(B=1, A=@Num.Num) {template="DISPATCH_RigVMDispatch_CoreEquals(in A,in B,out Result)"}.Result
   )
   RigVMFunction_ControlFlowBranch = &RigVMFunction_ControlFlowBranch(ExecuteContext=@n1.ExecuteContext, Condition=@RigVMFunction_MathBoolAnd.Result)
   /Script/ControlRig.RigUnit_SetDefaultMatchArray(
    Default=/Script/ControlRig.RigUnit_CollectionChildrenArray(
     TypeToSearch=Bone
     bDefaultChildren=True
     bIncludeParent=true
     bRecursive=true
     Parent=RigVMDispatchNode(Index=0, Array=@n2.Result) {template="DISPATCH_RigVMDispatch_ArrayGetAtIndex(in Array,in Index,out Element)"}.Element
    ) {is=struct RigVMUnitNode}.Items
    ExecutePin=@RigVMFunction_ControlFlowBranch.True
   ) {is=struct RigVMUnitNode, template="Set Default Match::Execute(in Default)"}
  ]}
]
hierarchy [
 HipRoot {type=Connector, loc=<0 0 0>, rot=<0 0 0 1>}
 Root {type=Socket, loc=<0 0 0>, rot=<0 0 0 1>}
 Tail {type=Connector, loc=<0 0 0>, rot=<0 0 0 1>}
 Parent {type=Connector, loc=<0 0 0>, rot=<0 0 0 1>}
 Orient Spaces {type=Connector, loc=<0 0 0>, rot=<0 0 0 1>}
]
variables [
 Base Name {type=FString, default=Tail FK}
 Connector Hip Root {type=FRigElementKey, default=(Type=Connector,Name="")}
 Connector Parent {type=FRigElementKey, default=(Type=Connector,Name="")}
 Connector Tail Base {type=FRigElementKey, default=(Type=Connector,Name="")}
 Connector Tail Tip {type=FRigElementKey, default=(Type=Connector,Name="")}
 Control Scale {type=double, default=1}
 Control Scale Factor Profile {
  type=FRuntimeFloatCurve
  default="(EditorCurveData=(Keys=((InterpMode=RCIM_Cubic,TangentMode=RCTM_User,Value=1.5),(InterpMode=RCIM_Cubic,TangentMode=RCTM_User,Time=0.75,Value=2.5),(InterpMode=RCIM_Cubic,TangentMode=RCTM_User,Time=1,Value=2)),DefaultValue=34028234663852885981170418348451692 ...(truncated)"
 }
 FK Control Rotation Offset {type=FRotator, default=(Pitch=0,Yaw=0,Roll=0)}
 Fk Controls {type="TArray<FRigElementKey>"}
 Fk Offset Controls {type="TArray<FRigElementKey>"}
 Fk Offset Controls Visibility {type=FRigElementKey, default=(Type=None,Name="")}
 Gimbal Control Name {type=FName, default=TailBase}
 Interacting Items {type="TArray<FRigElementKey>"}
 Is Interacting {type=bool, default=False}
 Position Parent Null {type=FRigElementKey, default=(Type=None,Name="")}
 Previous Euler Rotation {type=FVector, default=(X=0,Y=0,Z=0)}
 Root Module Color Weight {type=double, default=0}
 Rotation Order {type=EEulerRotationOrder, default=XZY}
 Tail Chain Joints {type="TArray<FRigElementKey>"}
 Tail Control Shape {
  type=FRigUnit_HierarchyAddControl_ShapeSettings
  default=(bVisible=True,Name="Circle_Thick",Color=(R=0,G=1,B=0,A=1),Transform=(Rotation=(X=0,Y=-0.7071,Z=0,W=0.7071),Translation=(X=0,Y=0,Z=0),Scale3D=(X=1,Y=1,Z=1)))
 }
 Tail Fk Offset Control Shape {
  type=FRigUnit_HierarchyAddControl_ShapeSettings
  default=(bVisible=True,Name="Square_Thick",Color=(R=0,G=0.325,B=1,A=1),Transform=(Rotation=(X=0,Y=-0.7071,Z=0,W=0.7071),Translation=(X=0,Y=0,Z=0),Scale3D=(X=1,Y=1,Z=1)))
 }
 Tail Gimbal Control {type=FRigElementKey, default=(Type=Bone,Name="")}
 Tail Gimbal Control Shape {
  type=FRigUnit_HierarchyAddControl_ShapeSettings
  default=(bVisible=True,Name="Circle_Thick",Color=(R=0,G=0.2245,B=0,A=1),Transform=(Rotation=(X=0,Y=-0.7071,Z=0,W=0.7071),Translation=(X=0,Y=0,Z=0),Scale3D=(X=7,Y=7,Z=7)))
 }
 Tail Rotate Control Shape {
  type=FRigUnit_HierarchyAddControl_ShapeSettings
  default=(bVisible=True,Name="",Color=(R=0,G=1,B=0,A=1),Transform=(Rotation=(X=0,Y=0,Z=0,W=1),Translation=(X=0,Y=0,Z=14),Scale3D=(X=1,Y=1,Z=1)))
 }
 Tail Rotate Feedback ProxyChannel {type=FRigElementKey, default=(Type=Bone,Name="")}
 Tail Rotate FeedbackThickness ProxyChannel {type=FRigElementKey, default=(Type=Bone,Name="")}
 Tail Rotate Position ProxyChannel {type=FRigElementKey, default=(Type=Bone,Name="")}
 Tail Rotate Proxy Control Parent {type=FRigElementKey, default=(Type=Bone,Name="")}
 Tail Rotate ProxyControl {type=FRigElementKey, default=(Type=Bone,Name="")}
 Tail Rotate Spread ProxyChannel {type=FRigElementKey, default=(Type=Bone,Name="")}
 Tail Space Nulls {type="TArray<FRigElementKey>"}
 Tail Spaces {type="TArray<FRigElementKey>"}
 Virtual Fk Joints {type="TArray<FRigElementKey>"}
]
